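; Parent quest-object Blueprint for touch-triggered cannon targets, presented as flaming rings that count toward a quest objective.
; When the quest initializes as valid, it activates the looping ring-of-fire cue.
; On objective success it stores the triggering player, spawns a burst effect and sound, hides itself, stops the loop and plays force feedback.
; Feedback goes to the first local player controller, not explicitly the triggering player, and both events print debug strings.
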(event EventQuestInitializedAndValid
  (|CallParentFunction)
  (Development|PrintString "Flaming Rings Initialized!" false false)
  (Components|Activation|Activate (Variables|Default|GetRing_Of_Fire_Loop_Cue) false))

(event EventObjectiveSuccessfullyCompleted
  (params (PlayerWhoTriggeredObject :type Actor%20Object%20Reference))
  (bind _output_get (Variables|Default|SetPlayerWhoTriggeredObject PlayerWhoTriggeredObject))
  (|CallParentFunction _output_get)
  (Development|PrintString "QUEST SUCCESS!!!!!!" false false)
  (Effects|SpawnEmitteratLocation (Variables|Default|GetFlamingBurstOnSuccess) (Transformation|GetActorLocation) (Transformation|GetActorRotation))
  (Rendering|SetActorHiddenInGame true)
  (Game|Feedback|ClientPlayForceFeedback (Utilities|Array|Get(acopy) (Fortnite|GetLocalFortPlayerControllers true)) (Variables|Default|GetForceFeedbackEffect) "None" false false false)
  (Audio|PlaySoundatLocation (Variables|Default|GetFlameBurstSound) (Transformation|GetActorLocation))
  (Components|Activation|Deactivate (Variables|Default|GetRing_Of_Fire_Loop_Cue)))

(fn ConstructionScript
  (|CallParentFunction))
